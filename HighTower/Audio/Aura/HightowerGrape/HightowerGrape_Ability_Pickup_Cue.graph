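# Sound cue for picking up the HightowerGrape ability, layering a generic teleporter pickup sound with an ability-specific pickup sound.
# A mixer plays the teleporter wave immediately and the ability wave after a short fixed delay.
# It uses weapon pickup attenuation and the loot-collect sound class at reduced volume, with overridden concurrency that blocks new instances rather than cutting playing ones.
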SoundCue /HighTower/Audio/Aura/HightowerGrape/HightowerGrape_Ability_Pickup_Cue
root=Mixer_0
nodes [
 {id=Delay_0, children [WavePlayer_1], properties {delayMax=0.2, delayMin=0.2}}
 {id=Mixer_0, children [WavePlayer_0,Delay_0], properties {inputVolume [1,1]}}
 {id=WavePlayer_0, properties {soundWaveAssetPtr=/Game/Sounds/Fort_Gadgets/Teleporter/teleporter_pickup_01}}
 {id=WavePlayer_1, properties {soundWaveAssetPtr=&/HightowerGrape_Ability_Pickup}}
]
properties {
 attenuationSettings=&Fort_Weapon_Pickup_Att
 bOverrideConcurrency=true
 maxDistance=3500
 soundClassObject=&Fort_Loot_Collect
 volumeMultiplier=0.1875
 concurrencyOverrides {resolutionRule=PreventNew}
}
