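# Sound cue for the HightowerGrape aura, layering a start sound with a looping sound that fades in.
# A slower envelope then lowers the combined volume over time.
# IsLocal and IsTeammate switches route that shared chain to an unattenuated mix for the local player, friendly attenuation for teammates, or louder default attenuation for others.
# It uses the player foley sound class and caps concurrency, stopping the quietest instance.
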SoundCue /HighTower/Audio/Aura/HightowerGrape/HightowerGrape_Aura_Loop_Cue
root=Switch_0
nodes [
 {id=Attenuation_1, children [Mixer_2], properties {attenuationSettings=/HighTower/Audio/Aura/Default/Hightower_Aura_Friendly_Att}}
 {id=Attenuation_2, children [Mixer_8], properties {attenuationSettings=/HighTower/Audio/Aura/Default/Hightower_Aura_Att}}
 {
  id=Enveloper_0
  children [WavePlayer_0]
  properties {pitchCurve {editorCurveData {keys [{time=0}]}}, volumeCurve {editorCurveData {keys [<at 0.0117 0.0072>,<at 0.5073 1>]}}}
 }
 {
  id=Enveloper_3
  children [Mixer_0]
  properties {pitchCurve {editorCurveData {keys [{time=0}]}}, volumeCurve {editorCurveData {keys [<at 0 1>,<at 10 0.2>]}}}
 }
 {id=Mixer_0, children [Enveloper_0,WavePlayer_3], properties {inputVolume [0.5,1]}}
 {id=Mixer_2, children [Enveloper_3], properties {inputVolume [0.5]}}
 {id=Mixer_6, children [Enveloper_3], properties {inputVolume [0.5]}}
 {id=Mixer_8, children [Enveloper_3], properties {inputVolume [1.5]}}
 {id=Switch_0, children [Switch_1,Switch_1,Mixer_6], properties {intParameterName=IsLocal}}
 {id=Switch_1, children [Attenuation_2, Attenuation_2, Attenuation_1], properties {intParameterName=IsTeammate}}
 {id=WavePlayer_0, properties {bLooping=true, soundWaveAssetPtr=&/HightowerGrape_Aura_Loop}}
 {id=WavePlayer_3, properties {soundWaveAssetPtr=&/HightowerGrape_Aura_Start}}
]
properties {
 bOverrideConcurrency=true
 maxDistance=650
 soundClassObject=&Fort_Player_Foley_Athena
 volumeMultiplier=1
 concurrencyOverrides {maxCount=3, resolutionRule=StopQuietest}
}
