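# HightowerGrape aura stop Sound Cue, which has no root or sound nodes and so produces no audio as authored.
# It keeps only playback settings: the third-person friendly player foley sound class, slightly reduced volume and a maximum distance.
# A concurrency override caps simultaneous instances and stops the quietest when the limit is exceeded.
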SoundCue /HighTower/Audio/Aura/HightowerGrape/HightowerGrape_Aura_Stop_Cue
root=null
nodes []
properties {
 bOverrideConcurrency=true
 maxDistance=4000
 soundClassObject=/Game/Sounds/FortSoundClassesAndModes/Fort_Player_Foley_3P_Friendly_Athena
 volumeMultiplier=0.9
 concurrencyOverrides {maxCount=3, resolutionRule=StopQuietest}
}
